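# Idle Niagara effect for the Scrap Tunnel backpack cosmetic, the variant with both spike sprite emitters disabled.
# Cube (GPU) and CubeLow (CPU, lower rate) spawn tumbling mesh particles split across three sockets, moved by curl noise and sprung toward their socket.
# Glow places blue sprites and lights on two other sockets, computing underglow alpha from camera angle against the owner's up axis.
# MeshCubes holds a persistent mesh at another socket.
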NiagaraSystem /BRCosmetics/Accessories/FORT_Backpacks/Backpack_F_MED_ScrapTunnel/FX/NS_Backpack_Scraptunnel_Idle_No_Spikes
dependencies {
 usedDataInterfaces [
  &NiagaraDataInterfaceArrayFloat
  &NiagaraDataInterfaceArrayInt32
  &NiagaraDataInterfaceArrayFloat3
  &NiagaraDataInterfaceCurve
  &NiagaraDataInterfaceCamera
  &NiagaraDataInterfaceSocketReader
 ]
 usedDynamicInputs [/Niagara/DynamicInputs/Divide/Divide_Int, &MakeIntFromFloat]
 usedModules [
  &self:SystemSpawnScript.SystemScriptSource.SystemScriptGraph.NiagaraNodeAssignment_0.SetVariables_#1
  &self:SystemSpawnScript.SystemScriptSource.SystemScriptGraph.NiagaraNodeAssignment_1.SetVariables_#2
  &SystemState
 ]
}
emitters {
 Cube {
  simTarget=GPUComputeSim
  emitterUpdateScript [
   &EmitterState()
   &SpawnRate(SpawnRate=$System.SpawnRate, Use Spawn Probability=false)
  ]
  particleSpawnScript [
   &PartitionParticles(ID to Partition:NiagaraInt32=$Engine.ExecIndex, Number Of Partitions=3, Partition Mode:ENiagara_PartitionMode_v2=::NewEnumerator5)
   &InitializeParticle(
    Color Maximum=<rgba 0.0781 0.0599 0.0545 1>
    Color Minimum=<rgba 0.1094 0.0641 0.0306 1>
    Color Mode=::NewEnumerator2
    Lifetime=&RandomRangeFloat(
     Minimum=&SelectFloatFromArray(
      Direct Array Index=$Particles.PartitionParticles.Partition
      Float Selection Array=DataInterface() {propertyValues {FloatData [0.3,1.5,0.45]}}
     )
     Maximum=&SelectFloatFromArray(
      Direct Array Index=$Particles.PartitionParticles.Partition
      Float Selection Array=DataInterface() {propertyValues {FloatData [0.75,2.5,0.75]}}
     )
    )
    Mesh Scale Mode=::NewEnumerator3
    Mesh Uniform Scale=&RandomRangeFloat(Minimum=0.015, Maximum=0.012)
    UsePositionOffset=false
    Write Material Random=true
   )
   &InitialMeshOrientation(Use Rotation Vector=false)
   &SocketLocation(
    SamplePosOffset=false
    Socket Index=&SelectIntFromArray(Direct Array Index=$Particles.PartitionParticles.Partition, Int Selection Array=DataInterface() {propertyValues {IntData [2,3,4]}})
    Socket Reader=$System.Socket Reader
    Socket Sampling Mode=::NewEnumerator3
    VelocitySpeedLimit=false
   )
   &ShapeLocation(MaskSpawnGroup=false, Override Seed=false, Sphere Radius=2, UseOwnerScale=true)
   &AddVelocity(
    Cone Axis=&SelectVectorFromArray(
     Direct Array Index=$Particles.PartitionParticles.Partition
     Vector Selection Array=DataInterface() {propertyValues {FloatData [<0.5 -1 -1>, <0.5 -1 1>, <-0.5 -1 1>]}}
    )
    Module.UseConeFalloff=false
    Module.UseVelDistribution=false
    Override Seed=false
    Velocity Mode=::NewEnumerator2
    Velocity Speed=&RandomRangeFloat(Minimum=10, Maximum=15)
   )
  ]
  particleUpdateScript [
   &ParticleState()
   &SocketLocation(
    Position Sampling=::NewEnumerator1
    SamplePosOffset=true
    Sampled Position Offset=$Particles.ShapeLocation.ShapeVector
    Socket Index=&SelectIntFromArray(Direct Array Index=$Particles.PartitionParticles.Partition, Int Selection Array=DataInterface() {propertyValues {IntData [2,3,4]}})
    Socket Reader=$System.Socket Reader
    Socket Sampling Mode=::NewEnumerator3
    VelocitySpeedLimit=false
   ) {module=SocketLocation001}
   &ScaleMeshSize(Scale Factor=&Multiply_VectorByFloat(
     Vector=<1 1 1>
     Float=&FloatFromCurve(FloatCurve=DataInterface() {propertyValues {Curve {
         defaultValue=1
         keys [
          {interpMode=RCIM_Cubic, tangentMode=RCTM_User, time=0, value=0}
          {arriveTangent=0.0748, interpMode=RCIM_Cubic, leaveTangent=0.0748, tangentMode=RCTM_User, time=0.2, value=1}
          {interpMode=RCIM_Cubic, tangentMode=RCTM_User, time=0.75, value=1}
          {tangentMode=RCTM_User, time=1, value=0}
         ]
        }}})
    ))
   &MeshRotationForce(Rotation=&RandomRangeVector(Minimum=<2 2 2>, Maximum=<1 1 1>))
   &Drag(
    Drag=2
    Rotational Drag:NiagaraFloat=&RandomRangeFloat(Minimum=0.2, Maximum=0.5)
    Use Linear Drag=true
    Use Rotational Drag=true
   )
   &CurlNoiseForce(
    Bias Noise=false
    Noise Frequency=100
    Noise Strength=&FloatFromCurve(
     FloatCurve=DataInterface() {propertyValues {Curve {defaultValue=1, keys [<at 0 0.1 RCIM_Cubic>,<at 1 1 RCIM_Cubic>]}}}
     Scale Curve=&SelectFloatFromArray(Direct Array Index=$Particles.PartitionParticles.Partition, Float Selection Array=DataInterface() {propertyValues {FloatData [25,15,25]}})
    )
    Pan Noise=true
    Pan Noise Field=<-1 0 0>
    Sample Position:NiagaraPosition=$Output.SocketLocation.SampledPosition
   )
   /Niagara/Modules/Update/Forces/SpringForce(
    " Particle Equilibrium Position":NiagaraPosition=$Output.SocketLocation001.SampledPosition
    Attachment Point Velocity:Vector3f=$Engine.Owner.Velocity
    Dampening Coefficient=2
    Desired Minimum Separation Distance=1
    Force Strength:NiagaraFloat=&RandomRangeFloat(Minimum=5, Maximum=10)
    Spring Tightness:NiagaraFloat=&RandomRangeFloat(Minimum=5, Maximum=10)
   )
   &SolveForcesAndVelocity(Clamp Velocity=false, Limit Acceleration=false)
  ]
  rendererClasses [&NiagaraMeshRendererProperties]
  renderers [{rendererClass=&NiagaraMeshRendererProperties, rendererIndex=0}]
 }
 CubeLow {
  emitterUpdateScript [
   &EmitterState()
   &SpawnRate(SpawnRate=&Multiply_Float(A=$System.SpawnRate, B=0.6), Use Spawn Probability=false)
  ]
  particleSpawnScript [
   &PartitionParticles(ID to Partition:NiagaraInt32=$Engine.ExecIndex, Number Of Partitions=3, Partition Mode:ENiagara_PartitionMode_v2=::NewEnumerator5)
   &InitializeParticle(
    Color Maximum=<rgba 0.0781 0.0599 0.0545 1>
    Color Minimum=<rgba 0.1094 0.0641 0.0306 1>
    Color Mode=::NewEnumerator2
    Lifetime=&RandomRangeFloat(
     Minimum=&SelectFloatFromArray(
      Direct Array Index=$Particles.PartitionParticles.Partition
      Float Selection Array=DataInterface() {propertyValues {FloatData [0.3,1.5,0.45]}}
     )
     Maximum=&SelectFloatFromArray(
      Direct Array Index=$Particles.PartitionParticles.Partition
      Float Selection Array=DataInterface() {propertyValues {FloatData [0.75,2.5,0.75]}}
     )
    )
    Mesh Scale Mode=::NewEnumerator3
    Mesh Uniform Scale=&RandomRangeFloat(Minimum=0.015, Maximum=0.012)
    UsePositionOffset=false
    Write Material Random=true
   )
   &InitialMeshOrientation(Use Rotation Vector=false)
   &SocketLocation(
    SamplePosOffset=false
    Socket Index=&SelectIntFromArray(Direct Array Index=$Particles.PartitionParticles.Partition, Int Selection Array=DataInterface() {propertyValues {IntData [2,3,4]}})
    Socket Reader=$System.Socket Reader
    Socket Sampling Mode=::NewEnumerator3
    VelocitySpeedLimit=false
   )
   &ShapeLocation(MaskSpawnGroup=false, Override Seed=false, Sphere Radius=2, UseOwnerScale=true)
   &AddVelocity(
    Cone Axis=&SelectVectorFromArray(
     Direct Array Index=$Particles.PartitionParticles.Partition
     Vector Selection Array=DataInterface() {propertyValues {FloatData [<0.5 -1 -1>, <0.5 -1 1>, <-0.5 -1 1>]}}
    )
    Module.UseConeFalloff=false
    Module.UseVelDistribution=false
    Override Seed=false
    Velocity Mode=::NewEnumerator2
    Velocity Speed=&RandomRangeFloat(Minimum=10, Maximum=15)
   )
  ]
  particleUpdateScript [
   &ParticleState()
   &SocketLocation(
    Position Sampling=::NewEnumerator1
    SamplePosOffset=true
    Sampled Position Offset=$Particles.ShapeLocation.ShapeVector
    Socket Index=&SelectIntFromArray(Direct Array Index=$Particles.PartitionParticles.Partition, Int Selection Array=DataInterface() {propertyValues {IntData [2,3,4]}})
    Socket Reader=$System.Socket Reader
    Socket Sampling Mode=::NewEnumerator3
    VelocitySpeedLimit=false
   ) {module=SocketLocation001}
   &ScaleMeshSize(Scale Factor=&Multiply_VectorByFloat(
     Vector=<1 1 1>
     Float=&FloatFromCurve(FloatCurve=DataInterface() {propertyValues {Curve {
         defaultValue=1
         keys [
          {interpMode=RCIM_Cubic, tangentMode=RCTM_User, time=0, value=0}
          {arriveTangent=0.0748, interpMode=RCIM_Cubic, leaveTangent=0.0748, tangentMode=RCTM_User, time=0.2, value=1}
          {interpMode=RCIM_Cubic, tangentMode=RCTM_User, time=0.75, value=1}
          {tangentMode=RCTM_User, time=1, value=0}
         ]
        }}})
    ))
   &MeshRotationForce(Rotation=&RandomRangeVector(Minimum=<2 2 2>, Maximum=<1 1 1>))
   &Drag(
    Drag=2
    Rotational Drag:NiagaraFloat=&RandomRangeFloat(Minimum=0.2, Maximum=0.5)
    Use Linear Drag=true
    Use Rotational Drag=true
   )
   &CurlNoiseForce(
    Bias Noise=false
    Noise Frequency=100
    Noise Strength=&FloatFromCurve(
     FloatCurve=DataInterface() {propertyValues {Curve {defaultValue=1, keys [<at 0 0.1 RCIM_Cubic>,<at 1 1 RCIM_Cubic>]}}}
     Scale Curve=&SelectFloatFromArray(Direct Array Index=$Particles.PartitionParticles.Partition, Float Selection Array=DataInterface() {propertyValues {FloatData [25,15,25]}})
    )
    Pan Noise=true
    Pan Noise Field=<-1 0 0>
    Sample Position:NiagaraPosition=$Output.SocketLocation.SampledPosition
   )
   &SpringForce(
    " Particle Equilibrium Position":NiagaraPosition=$Output.SocketLocation001.SampledPosition
    Attachment Point Velocity:Vector3f=$Engine.Owner.Velocity
    Dampening Coefficient=2
    Desired Minimum Separation Distance=1
    Force Strength:NiagaraFloat=&RandomRangeFloat(Minimum=5, Maximum=10)
    Spring Tightness:NiagaraFloat=&RandomRangeFloat(Minimum=5, Maximum=10)
   )
   &SolveForcesAndVelocity(Clamp Velocity=false, Limit Acceleration=false)
  ]
  rendererClasses [&NiagaraMeshRendererProperties]
  renderers [{rendererClass=&NiagaraMeshRendererProperties, rendererIndex=0}]
 }
 Glow {
  emitterUpdateScript [
   &EmitterState(
    Inactive Response=::NewEnumerator1
    Life Cycle Mode=::NewEnumerator1
    Loop Behavior=::NewEnumerator1
    Loop Duration Mode=::NewEnumerator1
    UseLoopDelay=false
   )
   &SpawnBurst_Instantaneous(Spawn Count=2, Use Spawn Probability=false, UseLoopCountLimit=false)
  ]
  particleSpawnScript [
   &PartitionParticles()
   &InitializeParticle(
    Color=<rgba 0.1042 0.5012 1 0.5>
    Sprite Size Mode=::NewEnumerator3
    Uniform Sprite Size=5
    UsePositionOffset=false
    Write Material Random=true
   )
   &SocketLocation(
    SamplePosOffset=true
    Sampled Position Offset=<0 0 -1.5>
    Socket Index=&SelectIntFromArray(Direct Array Index=$Particles.PartitionParticles.Partition, Int Selection Array=DataInterface() {propertyValues {IntData [0,1]}})
    Socket Reader=$System.Socket Reader
    Socket Sampling Mode=::NewEnumerator2
    VelocitySpeedLimit=false
   )
   /SetVariables(
    Particles.GlowCamPos=1
    Particles.UnderGlowAlignment:Vector3f=<0 1 0>
    Particles.UnderGlowCamPos=10
    Particles.UnderGlowFacing:Vector3f=<0 0 -1>
    Particles.UnderGlowSize:Vector2f=<5 5>
   ) {isSetParameters=true}
   /SetVariables(Particles.UnderglowPos:NiagaraPosition=&TransformPosition(
     Position=&AddVectorToPosition(Position=$Output.SocketLocation.SampledPosition, Vector=<0 0 -0.8>)
     Source Space=::Local
     Destination Space=::Simulation
    )) {isSetParameters=true}
  ]
  particleUpdateScript [
   &ParticleState(Kill Particles When Lifetime Has Elapsed=false, Let Infinitely Lived Particles Die When Emitter Deactivates=true)
   &ScaleColor(ScaleA=true, ScaleRGB=true)
   /SetVariables(Particles.UnderGlowColor:LinearColor=&MakeLinearColorFromVectorAndFloat(
     "Vector (RGB)"=&MakeVectorFromLinearColorRGB(Color=$Particles.Color)
     "Float (Alpha)"=&Lerp_Float(
      A=0.25
      B=0
      Alpha=&Power(
       a=2
       b=&NormalizeFloat(
        Value=&DotProduct(
         Vector A=$Engine.Owner.SystemZAxis
         Vector B=&ReturnCameraVector(Camera Vectors=::NewEnumerator4, Coordinate Space=::Simulation, Origin Position=$Particles.Position)
        )
        Min=0
        Max=1
       )
      )
     )
    )) {isSetParameters=true}
   &FN_Light_Attributes(
    Color=<rgba 0 0.3924 1 1>
    Create Color Attribute=true
    Create Exponent Attribute=true
    Create Position Attribute=true
    Create Radius Attribute=true
    Exponent=2
    Position=&TransformPosition(
     Position=&AddVectorToPosition(Position=$Particles.Position, Vector=<0 0 -10>)
     Source Space=::Local
     Destination Space=::Simulation
    )
    Radius=15
    Scale Color=20
   )
  ]
  rendererClasses [&NiagaraSpriteRendererProperties, &NiagaraLightRendererProperties]
  renderers [
   {rendererClass=&NiagaraSpriteRendererProperties}
   {rendererClass=&NiagaraSpriteRendererProperties, rendererIndex=1}
   {rendererClass=&NiagaraLightRendererProperties, rendererIndex=2}
  ]
 }
 MeshCubes {
  emitterUpdateScript [
   &EmitterState(
    Inactive Response=::NewEnumerator1
    Life Cycle Mode=::NewEnumerator1
    Loop Behavior=::NewEnumerator1
    Loop Duration Mode=::NewEnumerator1
    UseLoopDelay=false
   )
   &SpawnBurst_Instantaneous(Use Spawn Probability=false, UseLoopCountLimit=false)
  ]
  particleSpawnScript [
   &InitializeParticle(Color=<rgba 0 0 0 1>, Mesh Scale Mode=::NewEnumerator3, UsePositionOffset=false, Write Material Random=true)
   &SocketLocation(SamplePosOffset=false, Socket Index=5, Socket Reader=$System.Socket Reader, Socket Sampling Mode=::NewEnumerator2, VelocitySpeedLimit=false)
  ]
  particleUpdateScript [&ParticleState(Kill Particles When Lifetime Has Elapsed=false, Let Infinitely Lived Particles Die When Emitter Deactivates=true)]
  rendererClasses [&NiagaraMeshRendererProperties]
  renderers [{rendererClass=&NiagaraMeshRendererProperties, rendererIndex=0}]
 }
 Spikes {
  bEnabled=false
  simTarget=GPUComputeSim
  emitterUpdateScript [
   &EmitterState()
   &SpawnRate(SpawnRate=30, Use Spawn Probability=false)
  ]
  particleSpawnScript [
   &PartitionParticles()
   &InitializeParticle(
    Color Mode=::NewEnumerator0
    Lifetime Mode=::NewEnumerator1
    Sprite Size=&RandomRangeVector2D(Minimum=<0.25 2.5>, Maximum=<0.5 3.5>)
    Sprite Size Mode=::NewEnumerator4
    UsePositionOffset=false
    Write Material Random=true
   )
   &SocketLocation(
    SamplePosOffset=true
    Sampled Position Offset=<0 -7 3>
    Socket Index=5
    Socket Reader=$System.Socket Reader
    Socket Sampling Mode=::NewEnumerator2
    VelocitySpeedLimit=false
   )
   &ShapeLocation(
    Cylinder Distribution=::NewEnumerator2
    Cylinder Height=12
    Cylinder Radius=4
    MaskSpawnGroup=false
    Offset=&SelectVectorFromArray(
     Direct Array Index=$Particles.PartitionParticles.Partition
     Vector Selection Array=DataInterface() {propertyValues {FloatData [<6 0 0>, <-6 0 0>]}}
    )
    Offset Mode=::NewEnumerator0
    Shape Primitive=::NewEnumerator1
    U Position=&RandomRangeFloat(Minimum=1.1, Maximum=0.4)
    UseOwnerScale=true
    V Position=&RandomRangeFloat(Minimum=0, Maximum=1)
   )
   /SetVariables(Particles.SpriteAlignment=$Particles.ShapeLocation.ShapeNormal) {isSetParameters=true, module=SetVariables_#3}
  ]
  particleUpdateScript [
   &ParticleState()
   &ScaleSpriteSize(Uniform Curve Sprite Scale=DataInterface() {propertyValues {Curve {
       defaultValue=0
       keys [
        {interpMode=RCIM_Cubic, tangentMode=RCTM_User, time=0, value=0}
        {arriveTangent=1.3333, interpMode=RCIM_Cubic, leaveTangent=1.3333, time=0.2, value=1}
        {interpMode=RCIM_Cubic, tangentMode=RCTM_User, time=0.75, value=1}
        {tangentMode=RCTM_User, time=1, value=0}
       ]
      }}})
  ]
  rendererClasses [&NiagaraSpriteRendererProperties]
  renderers [{rendererClass=&NiagaraSpriteRendererProperties}]
 }
 SpikesLow {
  bEnabled=false
  emitterUpdateScript [
   &EmitterState()
   &SpawnRate(SpawnRate=15, Use Spawn Probability=false)
  ]
  particleSpawnScript [
   &PartitionParticles()
   &InitializeParticle(
    Color Mode=::NewEnumerator0
    Lifetime Mode=::NewEnumerator1
    Sprite Size=&RandomRangeVector2D(Minimum=<0.25 2.5>, Maximum=<0.5 3.5>)
    Sprite Size Mode=::NewEnumerator4
    UsePositionOffset=false
    Write Material Random=true
   )
   &SocketLocation(
    SamplePosOffset=true
    Sampled Position Offset=<0 -7 3>
    Socket Index=5
    Socket Reader=$System.Socket Reader
    Socket Sampling Mode=::NewEnumerator2
    VelocitySpeedLimit=false
   )
   &ShapeLocation(
    Cylinder Distribution=::NewEnumerator2
    Cylinder Height=12
    Cylinder Radius=4
    MaskSpawnGroup=false
    Offset=&SelectVectorFromArray(
     Direct Array Index=$Particles.PartitionParticles.Partition
     Vector Selection Array=DataInterface() {propertyValues {FloatData [<6 0 0>, <-6 0 0>]}}
    )
    Offset Mode=::NewEnumerator0
    Shape Primitive=::NewEnumerator1
    U Position=&RandomRangeFloat(Minimum=1.1, Maximum=0.4)
    UseOwnerScale=true
    V Position=&RandomRangeFloat(Minimum=0, Maximum=1)
   )
   /SetVariables(Particles.SpriteAlignment=$Particles.ShapeLocation.ShapeNormal) {isSetParameters=true, module=SetVariables_#3}
  ]
  particleUpdateScript [
   &ParticleState()
   &ScaleSpriteSize(Uniform Curve Sprite Scale=DataInterface() {propertyValues {Curve {
       defaultValue=0
       keys [
        {interpMode=RCIM_Cubic, tangentMode=RCTM_User, time=0, value=0}
        {arriveTangent=1.3333, interpMode=RCIM_Cubic, leaveTangent=1.3333, time=0.2, value=1}
        {interpMode=RCIM_Cubic, tangentMode=RCTM_User, time=0.75, value=1}
        {tangentMode=RCTM_User, time=1, value=0}
       ]
      }}})
  ]
  rendererClasses [&NiagaraSpriteRendererProperties]
  renderers [{rendererClass=&NiagaraSpriteRendererProperties}]
 }
}
summary {
 systemName=NS_Backpack_Scraptunnel_Idle_No_Spikes
 emitters [
  {emitterName=Cube, simTarget=GPUComputeSim, rendererClasses [&NiagaraMeshRendererProperties]}
  {emitterName=Glow, rendererClasses [&NiagaraSpriteRendererProperties, &NiagaraLightRendererProperties]}
  {emitterName=MeshCubes, rendererClasses [&NiagaraMeshRendererProperties]}
  {bEnabled=false, emitterName=Spikes, simTarget=GPUComputeSim, rendererClasses [&NiagaraSpriteRendererProperties]}
  {bEnabled=false, emitterName=SpikesLow, rendererClasses [&NiagaraSpriteRendererProperties]}
  {emitterName=CubeLow, rendererClasses [&NiagaraMeshRendererProperties]}
 ]
 userVariables []
}
